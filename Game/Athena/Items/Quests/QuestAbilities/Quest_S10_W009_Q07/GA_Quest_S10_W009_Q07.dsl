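; Quest gameplay ability for S10 Week 9 Q07 that credits destroying a Loot Carrier from a distance (50m, per the stat tag).
; It responds only to events whose target tags pass a comparison with the LootCarrier tag variable.
; On activation it measures the owning pawn's distance to the target and debug-prints it.
; If that meets the threshold variable, it sends the custom stat event to the quest manager; either way, the ability ends.
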(event Ability|EventActivateAbilityFromEvent
  (params (EventData :type Gameplay%20Event%20Data%20Structure%20%28by%20ref%29))
  (bind (_eventtag _instigator _target) (Utilities|Struct|BreakGameplayEventData EventData))
  (bind _owningplayerpawn (Variables|GAVars|Data|GetOwningPlayerPawn))
  (bind _distance (Class|AbilityFunctions|GetDistanceBetweenActors _owningplayerpawn _target))
  (|CallParentFunction EventData)
  (Development|PrintString (Utilities|String|ToString(Float) _distance) true true "(R=0.000000,G=0.660000,B=1.000000,A=1.000000)" 4e+01)
  (if (>= _distance (Variables|Default|GetDistance%20From%20LootCarrier))
    (Quest|SendComplexCustomStatEvent (Variables|GAVars|Data|GetOwningPlayerQuestManager) 0 "(GameplayTags=((TagName=\"Athena.Quests.Destroy.LootCarrier.50m\")))" "(GameplayTags=)" 1)
    (Ability|EndAbility)
    (else
      (Ability|EndAbility))))

(fn ShouldAbilityRespondToEvent
  (params (ActorInfo :type Gameplay%20Ability%20Actor%20Info%20Structure) (Payload_EventTag :type Gameplay%20Tag%20Structure) (Payload_Instigator :type Actor%20Object%20Reference) (Payload_Target :type Actor%20Object%20Reference) (Payload_OptionalObject :type Object%20Reference) (Payload_OptionalObject2 :type Object%20Reference) (Payload_ContextHandle :type Gameplay%20Effect%20Context%20Handle%20Structure) (Payload_InstigatorTags :type Gameplay%20Tag%20Container%20Structure) (Payload_TargetTags :type Gameplay%20Tag%20Container%20Structure) (Payload_EventMagnitude :type Float%20%28single-precision%29) (Payload_TargetData :type Gameplay%20Ability%20Target%20Data%20Handle%20Structure) (Payload_InstancedEventData :type Instanced%20Struct%20Structure))
  (return (GameplayTags|TagContainer|CompareTagContainertoOtherTags (Variables|Default|GetT_LootCarrier) Payload_TargetTags)))
